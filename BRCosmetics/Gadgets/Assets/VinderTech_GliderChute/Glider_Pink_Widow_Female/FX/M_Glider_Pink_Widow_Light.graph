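# Additive unlit material for the light effect on the Pink Widow female glider, usable on Niagara mesh particles.
# A soft box gradient mask builds the glow: red tinted by particle color plus faint untinted green, blended by V2 toward tinted blue and passed through an emissive adjustment function.
# Vertex color times particle alpha forms a second term, with V2 also blending in a view-angle fade.
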Material /BRCosmetics/Gadgets/Assets/VinderTech_GliderChute/Glider_Pink_Widow_Female/FX/M_Glider_Pink_Widow_Light
graph {
 properties {
  bUsedWithNiagaraMeshParticles=true
  blendMode=BLEND_Additive
  pixelDepthOffsetMode=PDOM_Legacy
  previewMesh=&/SM_Glider_Pink_Widow_Light
  shadingModel=MSM_Unlit
  usedShadingModels=Unlit
 }
 nodes [
  n1 = ParticleColor()
  n2 = Multiply(A=VertexColor(), B=@n1.A)
  n3 = ScalarParameter() {parameterName=FadeBasedOnViewAngle}
  n4 = ScalarParameter() {parameterName=V2}
  LinearInterpolate(
   A=@n2
   B=Multiply(
    A=@n2
    B=MaterialFunctionCall("FadeSensitivity (optional)"=@n3) {materialFunction=&FadeBasedOnViewAngle}.Result
   )
   Alpha=@n4
  )
  n5 = TextureSample() {
   samplerType=SAMPLERTYPE_Masks
   texture=&/T_Square_Box_Gradient_Mask_Soft
  }
  n6 = Multiply(A=@n5.G) {constB=0.1}
  n7 = Multiply(A=@n5.R, B=@n1.RGB)
  MaterialFunctionCall(Emissive=LinearInterpolate(A=Add(A=@n6, B=@n7), B=Multiply(A=@n5.B, B=@n1.RGB), Alpha=@n4)) {materialFunction=&MF_FX_EmissiveAdjustments}
 ]
}
