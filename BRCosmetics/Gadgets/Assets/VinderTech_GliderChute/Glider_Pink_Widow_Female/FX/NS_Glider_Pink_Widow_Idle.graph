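# Idle VFX for the Pink Widow glider: skeletal-mesh-attached fire mesh, glow sprites and lights, speed lines, and anim-trail ribbons.
# Looping fire and light particles flicker on curves, with intensity smoothed from ForwardDot or timeline time and some sprites scaled by time of day.
# Speed lines fade with ForwardDot, while four bone-pair ribbon emitters fade by a system integrator.
# Front-end modifiers switch lifetimes, alpha and trail velocity, and variant color parameters tint everything except the white speed lines.
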NiagaraSystem /BRCosmetics/Gadgets/Assets/VinderTech_GliderChute/Glider_Pink_Widow_Female/FX/NS_Glider_Pink_Widow_Idle
dependencies {
 usedDataInterfaces [
  &NiagaraDataInterfaceCurve
  &NiagaraDataInterfaceVector2DCurve
  &NiagaraDataInterfaceColorCurve
  &NiagaraDataInterfaceArrayFloat3
  &NiagaraDataInterfaceVectorCurve
  &NiagaraDataInterfaceArrayFloat
  &NiagaraDataInterfaceArrayFloat2
  &NiagaraDataInterfaceArrayColor
  &NiagaraDataInterfaceArrayInt32
 ]
 usedDynamicInputs [&Boolean_AndOperation]
 usedModules [
  &SystemState
  &FN_Integrator
  &Timeline
  &FN_Timer
 ]
}
emitters {
 Boost_light {
  emitterUpdateScript [
   &EmitterState()
   &SpawnBurst_Instantaneous(Age:NiagaraFloat=$Emitter.Age, Spawn Count=2, Use Spawn Probability=false, UseLoopCountLimit=false)
   &InterpolateOverTime(Rate Of Change=4, Target Float Value:NiagaraFloat=&ClampFloat(Float=$User.ForwardDot))
  ]
  particleSpawnScript [
   &InitializeParticle(
    Color=&MultiplyLinearColor(Linear Color=$User.Variant_Light, Scale Factor=<10 10 10 1>)
    Lifetime=&FN_FrontEndFloatModifier(Gameplay=1, Front End=0.35)
    Sprite Size Mode=::NewEnumerator3
    UsePositionOffset=true
    Write Material Random=true
   )
   &FN_SetSpawnGroup(Number of Spawn Groups=2)
  ]
  particleUpdateScript [
   &ParticleState(
    Kill Particles When Lifetime Has Elapsed=false
    Let Infinitely Lived Particles Die When Emitter Deactivates=true
    Loop Particles Lifetime=true
   )
   &SkeletalMeshLocation(
    Bone Sampling Mode=::NewEnumerator4
    MaskSpawnGroup=false
    Sampled Position Offset=&SelectVectorFromArray(Direct Array Index=$Particles.SpawnGroup, Vector Selection Array=DataInterface() {propertyValues {FloatData [<-7 50 -15>, <7 50 -15>]}})
    Skeletal Mesh=$User.SkeletalMesh
    Use Sampled Mesh Attribute as Ribbon ID=false
   )
   &ScaleColor(
    Scale Alpha=&FN_FrontEndFloatModifier(
     Gameplay=&Multiply_Float(
      A=$Emitter.InterpolateOverTime.InterpolateOverTime_Float.Moving Average
      B=&FloatFromCurve(
       FloatCurve=DataInterface() {propertyValues {Curve {keys [
           {interpMode=RCIM_Cubic, tangentMode=RCTM_User, time=0, value=1}
           {interpMode=RCIM_Cubic, tangentMode=RCTM_User, time=0.1, value=1.5}
           {arriveTangent=-1.25, interpMode=RCIM_Cubic, leaveTangent=-1.25, time=0.1568, value=1}
           {arriveTangent=-0.4779, interpMode=RCIM_Cubic, leaveTangent=-0.4779, time=0.3, value=1.25}
           {arriveTangent=0.6454, interpMode=RCIM_Cubic, leaveTangent=0.6454, time=0.3661, value=0.9}
           {arriveTangent=0.3821, interpMode=RCIM_Cubic, leaveTangent=0.3821, time=0.5324, value=1.4}
           {arriveTangent=-0.3158, interpMode=RCIM_Cubic, leaveTangent=-0.3158, time=0.6278, value=1}
           <at 0.6907 1.35 RCIM_Cubic>
           {arriveTangent=0.7168, interpMode=RCIM_Cubic, leaveTangent=0.7168, time=0.8247, value=1}
           <at 0.9 1.5 RCIM_Cubic>
           <at 1 1 RCIM_Cubic>
          ]}}}
       Scale Curve=0.15
      )
     )
     Front End=1
    )
    Scale RGB=&FN_FrontEndVector3Modifier(
     Gameplay=<1 1 1>
     Front End=&VectorFromFloat(Value=&FloatFromCurve(FloatCurve=DataInterface() {propertyValues {Curve {keys [
           {interpMode=RCIM_Cubic, tangentMode=RCTM_User, time=0, value=0}
           {arriveTangent=2, interpMode=RCIM_Cubic, leaveTangent=2, time=0.1189, value=0.8334}
           {arriveTangent=-0.6444, interpMode=RCIM_Cubic, leaveTangent=-0.6444, time=0.5, value=1}
           {arriveTangent=-2, interpMode=RCIM_Cubic, leaveTangent=-2, time=0.8877, value=0.338}
           {tangentMode=RCTM_User, time=1, value=0}
          ]}}}))
    )
    ScaleA=true
    ScaleRGB=true
   )
   /SetVariables(Particles.LightRadius=&FN_FrontEndFloatModifier(
     Gameplay=&MakeCustomFloatFromBool(
      Bool=&SetBoolByFloatComparison(A=$Emitter.InterpolateOverTime.InterpolateOverTime_Float.Moving Average, B=0.01, Comparison Type=::NewEnumerator6)
      True Float=0
      False Float=70
     )
     Front End=70
    )) {isSetParameters=true}
  ]
  rendererClasses [&NiagaraLightRendererProperties]
  renderers [{rendererClass=&NiagaraLightRendererProperties, rendererIndex=0}]
 }
 FireMesh {
  emitterUpdateScript [
   &EmitterState()
   &SpawnBurst_Instantaneous(Spawn Probability=HlslExpression() {hlslExpression="Emitter.LoopCount == 0"}, Use Spawn Probability=true, UseLoopCountLimit=false)
   &InterpolateOverTime(Rate Of Change=4, Target Float Value:NiagaraFloat=&ClampFloat(Float=$System.Timeline.Time))
  ]
  particleSpawnScript [
   &InitializeParticle(
    Color=&MultiplyLinearColor(Linear Color=$User.Variant_Light, Scale Factor=<2 2 2 5>)
    Lifetime=&FN_FrontEndFloatModifier(Gameplay=1, Front End=0.5)
    Mesh Scale Mode=::NewEnumerator3
    UsePositionOffset=false
    Write Material Random=true
   )
   &CameraOffset()
  ]
  particleUpdateScript [
   &ParticleState(
    Kill Particles When Lifetime Has Elapsed=false
    Let Infinitely Lived Particles Die When Emitter Deactivates=true
    Loop Particles Lifetime=true
   )
   &SkeletalMeshLocation(Bone Sampling Mode=::NewEnumerator4, MaskSpawnGroup=false, Sampled Position Offset=<0 20 -7>, Skeletal Mesh=$User.SkeletalMesh)
   &ScaleColor(
    Scale Alpha=&Lerp_Float(
     A=&FN_FrontEndFloatModifier(Gameplay=0, Front End=1)
     B=&FloatFromCurve(
      FloatCurve=DataInterface() {propertyValues {Curve {keys [
          {interpMode=RCIM_Cubic, tangentMode=RCTM_User, time=0, value=1}
          {interpMode=RCIM_Cubic, tangentMode=RCTM_User, time=0.1, value=1.5}
          {arriveTangent=-1.25, interpMode=RCIM_Cubic, leaveTangent=-1.25, time=0.1568, value=1}
          {arriveTangent=-0.4779, interpMode=RCIM_Cubic, leaveTangent=-0.4779, time=0.3, value=1.25}
          {arriveTangent=0.6454, interpMode=RCIM_Cubic, leaveTangent=0.6454, time=0.3661, value=0.9}
          {arriveTangent=0.3821, interpMode=RCIM_Cubic, leaveTangent=0.3821, time=0.5324, value=1.4}
          {arriveTangent=-0.3158, interpMode=RCIM_Cubic, leaveTangent=-0.3158, time=0.6278, value=1}
          <at 0.6907 1.35 RCIM_Cubic>
          {arriveTangent=0.7168, interpMode=RCIM_Cubic, leaveTangent=0.7168, time=0.8247, value=1}
          <at 0.9 1.5 RCIM_Cubic>
          <at 1 1 RCIM_Cubic>
         ]}}}
      Scale Curve=5
     )
     Alpha=$Emitter.InterpolateOverTime.InterpolateOverTime_Float.Moving Average
    )
    Scale RGB=&Multiply_Vector(
     A=&FN_FrontEndVector3Modifier(
      Gameplay=<1 1 1>
      Front End=&VectorFromFloat(Value=&FloatFromCurve(
        FloatCurve=DataInterface() {propertyValues {Curve {keys [
            {interpMode=RCIM_Cubic, tangentMode=RCTM_User, time=0, value=0}
            {arriveTangent=2, interpMode=RCIM_Cubic, leaveTangent=2, time=0.1189, value=0.8334}
            {arriveTangent=-0.6444, interpMode=RCIM_Cubic, leaveTangent=-0.6444, time=0.5, value=1}
            {arriveTangent=-2, interpMode=RCIM_Cubic, leaveTangent=-2, time=0.8877, value=0.338}
            {tangentMode=RCTM_User, time=1, value=0}
           ]}}}
        Scale Curve=10
       ))
     )
     B=&VectorFromFloat(Value=&FloatFromCurve(FloatCurve=DataInterface() {propertyValues {Curve {keys [
           {interpMode=RCIM_Cubic, tangentMode=RCTM_User, time=0, value=1}
           {interpMode=RCIM_Cubic, tangentMode=RCTM_User, time=0.5, value=2}
           {tangentMode=RCTM_User, time=1, value=1}
          ]}}}))
    )
    ScaleA=true
    ScaleRGB=true
   )
  ]
  rendererClasses [&NiagaraMeshRendererProperties]
  renderers [{rendererClass=&NiagaraMeshRendererProperties, rendererIndex=0}]
 }
 FireMesh001 {
  emitterUpdateScript [
   &EmitterState()
   &SpawnBurst_Instantaneous(
    Spawn Count=2
    Spawn Probability=HlslExpression() {hlslExpression="Emitter.LoopCount == 0"}
    Use Spawn Probability=true
    UseLoopCountLimit=false
   )
   &InterpolateOverTime(Rate Of Change=4, Target Float Value:NiagaraFloat=&ClampFloat(Float=$User.ForwardDot))
  ]
  particleSpawnScript [
   &InitializeParticle(
    Color=&MultiplyLinearColor(Linear Color=$User.Variant_Light, Scale Factor=<1.5 1.5 1.5 1>)
    Lifetime=&FN_FrontEndFloatModifier(Gameplay=1, Front End=0.3)
    Sprite Size=<30 35>
    Sprite Size Mode=::NewEnumerator4
    UsePositionOffset=false
    Write Material Random=true
   )
   &CameraOffset(Camera Offset Amount=10)
   &FN_SetSpawnGroup(Number of Spawn Groups=2)
  ]
  particleUpdateScript [
   &ParticleState(
    Kill Particles When Lifetime Has Elapsed=false
    Let Infinitely Lived Particles Die When Emitter Deactivates=true
    Loop Particles Lifetime=true
   )
   &SkeletalMeshLocation(
    Bone Sampling Mode=::NewEnumerator4
    MaskSpawnGroup=false
    Sampled Position Offset=&SelectVectorFromArray(Direct Array Index=$Particles.SpawnGroup, Vector Selection Array=DataInterface() {propertyValues {FloatData [<-6 40 -32>, <6 40 -32>]}})
    Skeletal Mesh=$User.SkeletalMesh
   )
   &ScaleColor(
    Scale Alpha=&Multiply_Float(
     A=&FNTimeOfDayScalar(Night Scale=0.3, Morning Scale=0.1, Day Scale=0.5, Evening Scale=0.05)
     B=&Lerp_Float(
      A=&FloatFromCurve(
       FloatCurve=DataInterface() {propertyValues {Curve {keys [
           {interpMode=RCIM_Cubic, tangentMode=RCTM_User, time=0, value=0.1}
           {interpMode=RCIM_Cubic, tangentMode=RCTM_User, time=0.5, value=0.2}
           {tangentMode=RCTM_User, time=1, value=0.1}
          ]}}}
       Scale Curve=0.25
      )
      B=&FloatFromCurve(FloatCurve=DataInterface() {propertyValues {Curve {keys [
           {interpMode=RCIM_Cubic, tangentMode=RCTM_User, time=0, value=1}
           {interpMode=RCIM_Cubic, tangentMode=RCTM_User, time=0.1, value=1.5}
           {arriveTangent=-1.25, interpMode=RCIM_Cubic, leaveTangent=-1.25, time=0.1568, value=1}
           {arriveTangent=-0.4779, interpMode=RCIM_Cubic, leaveTangent=-0.4779, time=0.3, value=1.25}
           {arriveTangent=0.6454, interpMode=RCIM_Cubic, leaveTangent=0.6454, time=0.3661, value=0.9}
           {arriveTangent=0.3821, interpMode=RCIM_Cubic, leaveTangent=0.3821, time=0.5324, value=1.4}
           {arriveTangent=-0.3158, interpMode=RCIM_Cubic, leaveTangent=-0.3158, time=0.6278, value=1}
           <at 0.6907 1.35 RCIM_Cubic>
           {arriveTangent=0.7168, interpMode=RCIM_Cubic, leaveTangent=0.7168, time=0.8247, value=1}
           <at 0.9 1.5 RCIM_Cubic>
           <at 1 1 RCIM_Cubic>
          ]}}})
      Alpha=$Emitter.InterpolateOverTime.InterpolateOverTime_Float.Moving Average
     )
    )
    Scale RGB=&FN_FrontEndVector3Modifier(
     Gameplay=<1 1 1>
     Front End=&VectorFromFloat(Value=&FloatFromCurve(
       FloatCurve=DataInterface() {propertyValues {Curve {keys [
           {interpMode=RCIM_Cubic, tangentMode=RCTM_User, time=0, value=0}
           {arriveTangent=2, interpMode=RCIM_Cubic, leaveTangent=2, time=0.1189, value=0.8334}
           {arriveTangent=-0.6444, interpMode=RCIM_Cubic, leaveTangent=-0.6444, time=0.5, value=1}
           {arriveTangent=-2, interpMode=RCIM_Cubic, leaveTangent=-2, time=0.8877, value=0.338}
           {tangentMode=RCTM_User, time=1, value=0}
          ]}}}
       Scale Curve=5
      ))
    )
    ScaleA=true
    ScaleRGB=true
   )
  ]
  rendererClasses [&NiagaraSpriteRendererProperties]
  renderers [{rendererClass=&NiagaraSpriteRendererProperties}]
 }
 FireMesh002 {
  emitterUpdateScript [
   &EmitterState()
   &SpawnBurst_Instantaneous(
    Spawn Count=2
    Spawn Probability=HlslExpression() {hlslExpression="Emitter.LoopCount == 0"}
    Use Spawn Probability=true
    UseLoopCountLimit=false
   )
   &InterpolateOverTime(Rate Of Change=4, Target Float Value:NiagaraFloat=&ClampFloat(Float=$System.Timeline.Time))
  ]
  particleSpawnScript [
   &InitializeParticle(
    Color=&MultiplyLinearColor(Linear Color=$User.Variant_Light, Scale Factor=<1 1 1 0.1>)
    Lifetime=&FN_FrontEndFloatModifier(Gameplay=1, Front End=0.35)
    Sprite Size=<100 100>
    Sprite Size Mode=::NewEnumerator4
    UsePositionOffset=false
    Write Material Random=true
   )
   &CameraOffset(Camera Offset Amount=10)
   &FN_SetSpawnGroup(Number of Spawn Groups=2)
  ]
  particleUpdateScript [
   &ParticleState(
    Kill Particles When Lifetime Has Elapsed=false
    Let Infinitely Lived Particles Die When Emitter Deactivates=true
    Loop Particles Lifetime=true
   )
   &SkeletalMeshLocation(
    Bone Sampling Mode=::NewEnumerator4
    MaskSpawnGroup=false
    Sampled Position Offset=&SelectVectorFromArray(Direct Array Index=$Particles.SpawnGroup, Vector Selection Array=DataInterface() {propertyValues {FloatData [<-6 40 -32>, <6 40 -32>]}})
    Skeletal Mesh=$User.SkeletalMesh
   )
   &ScaleColor(
    Scale Alpha=&Multiply_Float(
     A=&FNTimeOfDayScalar(Night Scale=0.5, Morning Scale=0.1, Evening Scale=0.2)
     B=&Lerp_Float(
      A=&FloatFromCurve(
       FloatCurve=DataInterface() {propertyValues {Curve {keys [
           {interpMode=RCIM_Cubic, tangentMode=RCTM_User, time=0, value=0.1}
           {interpMode=RCIM_Cubic, tangentMode=RCTM_User, time=0.5, value=0.2}
           {tangentMode=RCTM_User, time=1, value=0.1}
          ]}}}
       Scale Curve=0.25
      )
      B=&FloatFromCurve(FloatCurve=DataInterface() {propertyValues {Curve {keys [
           {interpMode=RCIM_Cubic, tangentMode=RCTM_User, time=0, value=1}
           {interpMode=RCIM_Cubic, tangentMode=RCTM_User, time=0.1, value=1.5}
           {arriveTangent=-1.25, interpMode=RCIM_Cubic, leaveTangent=-1.25, time=0.1568, value=1}
           {arriveTangent=-0.4779, interpMode=RCIM_Cubic, leaveTangent=-0.4779, time=0.3, value=1.25}
           {arriveTangent=0.6454, interpMode=RCIM_Cubic, leaveTangent=0.6454, time=0.3661, value=0.9}
           {arriveTangent=0.3821, interpMode=RCIM_Cubic, leaveTangent=0.3821, time=0.5324, value=1.4}
           {arriveTangent=-0.3158, interpMode=RCIM_Cubic, leaveTangent=-0.3158, time=0.6278, value=1}
           <at 0.6907 1.35 RCIM_Cubic>
           {arriveTangent=0.7168, interpMode=RCIM_Cubic, leaveTangent=0.7168, time=0.8247, value=1}
           <at 0.9 1.5 RCIM_Cubic>
           <at 1 1 RCIM_Cubic>
          ]}}})
      Alpha=$Emitter.InterpolateOverTime.InterpolateOverTime_Float.Moving Average
     )
    )
    Scale RGB=&FN_FrontEndVector3Modifier(
     Gameplay=<1 1 1>
     Front End=&VectorFromFloat(Value=&FloatFromCurve(
       FloatCurve=DataInterface() {propertyValues {Curve {keys [
           {interpMode=RCIM_Cubic, tangentMode=RCTM_User, time=0, value=0}
           {arriveTangent=2, interpMode=RCIM_Cubic, leaveTangent=2, time=0.1189, value=0.8334}
           {arriveTangent=-0.6444, interpMode=RCIM_Cubic, leaveTangent=-0.6444, time=0.5, value=1}
           {arriveTangent=-2, interpMode=RCIM_Cubic, leaveTangent=-2, time=0.8877, value=0.338}
           {tangentMode=RCTM_User, time=1, value=0}
          ]}}}
       Scale Curve=5
      ))
    )
    ScaleA=true
    ScaleRGB=true
   )
  ]
  rendererClasses [&NiagaraSpriteRendererProperties]
  renderers [{rendererClass=&NiagaraSpriteRendererProperties}]
 }
 SpeedLine {
  emitterUpdateScript [
   &EmitterState()
   &SpawnRate(
    SpawnRate=&MakeCustomFloatFromBool(Bool=$User.bIsFrontEndPreview, True Float=50, False Float=90)
    Use Spawn Probability=true
   ) {module=SpawnRate001}
  ]
  particleSpawnScript [
   &InitializeParticle(
    Color=&MakeLinearColorFromVectorAndFloat(
     "Vector (RGB)"=<1 1 1>
     "Float (Alpha)"=&Multiply_Float(
      A=&UniformRangedFloat(Minimum=0.1, Maximum=0.18)
      B=&MakeCustomFloatFromBool(Bool=$User.bIsFrontEndPreview, True Float=0.5, False Float=1)
     )
    )
    Lifetime Max=0.16
    Lifetime Min=0.1
    Lifetime Mode=::NewEnumerator1
    Sprite Size Max=<3 10>
    Sprite Size Min=<2 2>
    Sprite Size Mode=::NewEnumerator2
    UsePositionOffset=false
    Write Material Random=true
   )
   &TorusLocation(
    Align Torus With Axis=true
    Handle Radius=0
    Large Radius:NiagaraFloat=&RandomRangeFloat(Minimum=80, Maximum=150)
    MaskSpawnGroup=false
    Module.Override Randomness=false
    Offset=<0 -180 0>
    Torus Axis=<0 1 0>
   )
   &ShapeLocation(
    Handle Radius=1
    Large Radius=&RandomRangeFloat(Minimum=80, Maximum=150)
    MaskSpawnGroup=false
    Non Uniform Scale=<1.5 1 0.1>
    Override Seed=false
    Shape Primitive=::NewEnumerator3
    UseOwnerScale=true
   )
   &AddVelocityFromPoint(Module.Use Falloff Distance=false, Origin Offset=<0 -500 0>, Velocity Strength=600)
   &AddVelocity(Velocity=&RandomRangeVector(Minimum=<0 300 0>, Maximum=<1 700 1>))
  ]
  particleUpdateScript [
   &SolveForcesAndVelocity(Clamp Velocity=false, Limit Acceleration=false)
   &ParticleState()
   &ScaleSpriteSize(Scale Factor=&Vector2DFromCurve(
     Vector2Curve=DataInterface() {propertyValues {XCurve {keys [<at 0 0.5 RCIM_Cubic>,<at 0.5 1 RCIM_Cubic>]}, YCurve {keys [<at 0 0 RCIM_Cubic>,<at 0.5 1 RCIM_Cubic>]}}}
     ScaleCurve=&Vector2DFromFloat(Value=1)
    ))
   &ScaleSpriteSizeBySpeed(Max Scale Factor=<1 50>, Min Scale Factor=<0.01 0.05>, Velocity Threshold=1)
   &ScaleColor(
    Scale Alpha=&FloatFromCurve(
     FloatCurve=DataInterface() {propertyValues {Curve {keys [
         <at 0 0 RCIM_Cubic>
         <at 0.3759 1.0207>
         {arriveTangent=-2.221, interpMode=RCIM_Cubic, leaveTangent=-2.221, tangentMode=RCTM_User, time=1, value=0}
        ]}}}
     Scale Curve=&MakeCustomFloatFromBool(
      Bool=$User.bIsFrontEndPreview
      True Float=&MakeCustomFloatFromBool(Bool=$User.bIsFullyDeployed, True Float=1, False Float=0)
      False Float=&Lerp_Float(A=0, B=1, Alpha=$User.ForwardDot)
     )
    )
    Scale RGB=&VectorFromFloat(Value=1)
    ScaleA=true
    ScaleRGB=true
   )
  ]
  rendererClasses [&NiagaraSpriteRendererProperties]
  renderers [{rendererClass=&NiagaraSpriteRendererProperties}]
 }
 Trails {
  emitterUpdateScript [
   &EmitterState()
   /Game/Effects/Niagara/Modules/AnimTrail/FN_SpawnAnimTrail/FN_SpawnAnimTrail(AnimTrailWidthMode:ENiagaraLegacyTrailWidthMode=::FromSecond, Number of Anim Trails=2, SpawnRate=64)
  ]
  particleSpawnScript [
   /Game/Effects/Niagara/Modules/AnimTrail/FN_InitializeAnimTrail/FN_InitializeAnimTrailBoneIndecies(First Bone or Socket Index=1)
   &FN_InitializeAnimTrailBoneIndecies(AnimTrail Selection=1, First Bone or Socket Index=3, Second Bone or Socket Index=2) {module=FN_InitializeAnimTrailBoneIndecies001}
   /Game/Effects/Niagara/Modules/AnimTrail/FN_InitializeAnimTrail/FN_InitializeAnimTrail(
    AnimTrailWidth=0.7
    Color=&MultiplyLinearColor(Linear Color=$User.Variant_Color, Scale Factor=<1 1 1 2>)
    Lifetime=0.15
   )
   &AddVelocity(Velocity=&FN_FrontEndVector3Modifier(Gameplay=<0 0 0>, Front End=<0 2000 0>))
   &AddVelocity(
    Coordinate Space=::World
    Velocity=&FN_FrontEndVector3Modifier(Gameplay=<0 0 300>, Front End=<0 0 0>)
   ) {module=AddVelocity001}
  ]
  particleUpdateScript [
   &InheritSourceMovement(Applied Position Scale=<0.25 0.25 0.8>, ApplyForce=false, ApplyPosition=true, ApplyPositionFalloff=false)
   &SolveForcesAndVelocity(Clamp Velocity=false, Limit Acceleration=false)
   /Game/Effects/Niagara/Modules/AnimTrail/FN_UpdateAnimTrail/FN_UpdateAnimTrail(New float=1)
   &ScaleColor(
    Scale Alpha=&FN_FrontEndFloatModifier(Gameplay=$System.FN_Integrator.Value, Front End=1)
    ScaleA=true
    ScaleRGB=true
   )
  ]
  rendererClasses [&NiagaraRibbonRendererProperties]
  renderers [{rendererClass=&NiagaraRibbonRendererProperties}]
 }
 Trails001 {
  emitterUpdateScript [
   &EmitterState()
   &FN_SpawnAnimTrail(Number of Anim Trails=2, SpawnRate=64)
  ]
  particleSpawnScript [
   &FN_InitializeAnimTrailBoneIndecies(First Bone or Socket Index=4, Second Bone or Socket Index=5)
   &FN_InitializeAnimTrailBoneIndecies(AnimTrail Selection=1, First Bone or Socket Index=6, Second Bone or Socket Index=7) {module=FN_InitializeAnimTrailBoneIndecies001}
   &FN_InitializeAnimTrail(
    AnimTrailWidth=0.6
    Color=&MultiplyLinearColor(Linear Color=$User.Variant_Color, Scale Factor=<1 1 1 1.5>)
    Lifetime=0.12
   )
   &AddVelocity(Velocity=&FN_FrontEndVector3Modifier(Gameplay=<0 0 0>, Front End=<0 2000 0>))
   &AddVelocity(
    Coordinate Space=::World
    Velocity=&FN_FrontEndVector3Modifier(Gameplay=<0 0 450>, Front End=<0 0 0>)
   ) {module=AddVelocity001}
  ]
  particleUpdateScript [
   &InheritSourceMovement(Applied Position Scale=<0.25 0.25 0.8>, ApplyForce=false, ApplyPosition=true, ApplyPositionFalloff=false)
   &SolveForcesAndVelocity(Clamp Velocity=false, Limit Acceleration=false)
   &FN_UpdateAnimTrail(New float=0.8)
   &ScaleColor(
    Scale Alpha=&FN_FrontEndFloatModifier(Gameplay=$System.FN_Integrator.Value, Front End=1)
    ScaleA=true
    ScaleRGB=true
   )
  ]
  rendererClasses [&NiagaraRibbonRendererProperties]
  renderers [{rendererClass=&NiagaraRibbonRendererProperties}]
 }
 Trails002 {
  emitterUpdateScript [
   &EmitterState()
   &FN_SpawnAnimTrail(Number of Anim Trails=2, SpawnRate=64)
  ]
  particleSpawnScript [
   &FN_InitializeAnimTrailBoneIndecies(First Bone or Socket Index=8, Second Bone or Socket Index=9)
   &FN_InitializeAnimTrailBoneIndecies(AnimTrail Selection=1, First Bone or Socket Index=10, Second Bone or Socket Index=11) {module=FN_InitializeAnimTrailBoneIndecies001}
   &FN_InitializeAnimTrail(
    AnimTrailWidth=0.6
    Color=&MultiplyLinearColor(Linear Color=$User.Variant_Color, Scale Factor=<1 1 1 1.25>)
    Lifetime=0.15
   )
   &AddVelocity(Velocity=&FN_FrontEndVector3Modifier(Gameplay=<0 0 0>, Front End=<0 2000 0>))
   &AddVelocity(
    Coordinate Space=::World
    Velocity=&FN_FrontEndVector3Modifier(Gameplay=<0 0 450>, Front End=<0 0 0>)
   ) {module=AddVelocity001}
  ]
  particleUpdateScript [
   &InheritSourceMovement(Applied Position Scale=<0.25 0.25 0.8>, ApplyForce=false, ApplyPosition=true, ApplyPositionFalloff=false)
   &SolveForcesAndVelocity(Clamp Velocity=false, Limit Acceleration=false)
   &FN_UpdateAnimTrail(New float=0.8)
   &ScaleColor(
    Scale Alpha=&FN_FrontEndFloatModifier(Gameplay=$System.FN_Integrator.Value, Front End=1)
    ScaleA=true
    ScaleRGB=true
   )
  ]
  rendererClasses [&NiagaraRibbonRendererProperties]
  renderers [{rendererClass=&NiagaraRibbonRendererProperties}]
 }
 Trails003 {
  emitterUpdateScript [
   &EmitterState()
   &FN_SpawnAnimTrail(AnimTrailWidthMode:ENiagaraLegacyTrailWidthMode=::FromSecond, Number of Anim Trails=2, SpawnRate=64)
  ]
  particleSpawnScript [
   &FN_InitializeAnimTrailBoneIndecies(First Bone or Socket Index=13, Second Bone or Socket Index=12)
   &FN_InitializeAnimTrailBoneIndecies(AnimTrail Selection=1, First Bone or Socket Index=15, Second Bone or Socket Index=14) {module=FN_InitializeAnimTrailBoneIndecies001}
   &FN_InitializeAnimTrail(
    AnimTrailWidth=0.5
    Color=&MultiplyLinearColor(Linear Color=$User.Variant_Color, Scale Factor=<1 1 1 1>)
    Lifetime=0.2
   )
   &AddVelocity(Velocity=&FN_FrontEndVector3Modifier(Gameplay=<0 0 0>, Front End=<0 2000 0>))
   &AddVelocity(
    Coordinate Space=::World
    Velocity=&FN_FrontEndVector3Modifier(Gameplay=<0 0 250>, Front End=<0 0 0>)
   ) {module=AddVelocity001}
  ]
  particleUpdateScript [
   &InheritSourceMovement(
    Applied Position Falloff Curve=DataInterface() {propertyValues {Curve {keys [
        {arriveTangent=9.0725, interpMode=RCIM_Cubic, leaveTangent=9.0725, tangentMode=RCTM_User, time=0, value=0}
        {arriveTangent=-2.5249, interpMode=RCIM_Cubic, leaveTangent=-2.5249, tangentMode=RCTM_User, time=0.1408, value=0.9847}
        {arriveTangent=-0.0405, interpMode=RCIM_Cubic, leaveTangent=-0.0405, tangentMode=RCTM_User, time=1, value=0}
       ]}}}
    Applied Position Falloff Distance=750
    Applied Position Scale=<0.5 0.5 1>
    ApplyForce=false
    ApplyPosition=true
    ApplyPositionFalloff=true
    ApplyPositionFalloffCurve=true
   )
   &DynamicMaterialParameters(
    Index 0 Param 1=1
    Index 0 Param 2=-2
    Index 0 Param 4=1
    Index 1 Param 2=1
    Index 3 Param 4=$System.FN_Timer001.Value
    Module.Param0WriteEnabled=true
    Module.Param0WriteEnabled1=true
    Module.Param0WriteEnabled3=true
    Module.Param1WriteEnabled=true
    Module.Param1WriteEnabled1=true
    Module.Param1WriteEnabled3=true
    Module.Param2WriteEnabled=true
    Module.Param2WriteEnabled1=true
    Module.Param2WriteEnabled3=true
    Module.Param3WriteEnabled=true
    Module.Param3WriteEnabled1=true
    Module.Param3WriteEnabled3=true
    Write Parameter Index 1=true
    Write Parameter Index 3=true
   ) {module=DynamicMaterialParameters002}
   &SolveForcesAndVelocity(Clamp Velocity=false, Limit Acceleration=false)
   &FN_UpdateAnimTrail(New float=1)
   &ScaleColor(
    Scale Alpha=&FN_FrontEndFloatModifier(Gameplay=$System.FN_Integrator.Value, Front End=1)
    ScaleA=true
    ScaleRGB=true
   )
  ]
  rendererClasses [&NiagaraRibbonRendererProperties]
  renderers [{rendererClass=&NiagaraRibbonRendererProperties}]
 }
}
summary {
 systemName=NS_Glider_Pink_Widow_Idle
 emitters [
  {emitterName=FireMesh, rendererClasses [&NiagaraMeshRendererProperties]}
  {emitterName=SpeedLine, rendererClasses [&NiagaraSpriteRendererProperties]}
  {emitterName=Boost_light, rendererClasses [&NiagaraLightRendererProperties]}
  {emitterName=FireMesh001, rendererClasses [&NiagaraSpriteRendererProperties]}
  {emitterName=Trails001, rendererClasses [&NiagaraRibbonRendererProperties]}
  {emitterName=Trails, rendererClasses [&NiagaraRibbonRendererProperties]}
  {emitterName=Trails002, rendererClasses [&NiagaraRibbonRendererProperties]}
  {emitterName=FireMesh002, rendererClasses [&NiagaraSpriteRendererProperties]}
  {emitterName=Trails003, rendererClasses [&NiagaraRibbonRendererProperties]}
 ]
 userVariables [
  User.bIsFrontEnd {type=NiagaraBool, defaultValue=true}
  User.bIsFrontEndPreview {type=NiagaraBool, defaultValue=true}
  User.bIsFullyDeployed {type=NiagaraBool, defaultValue=true}
  User.BPOpacityMult {type=NiagaraFloat, defaultValue=0, description=""}
  User.ColorIndex {type=NiagaraFloat, defaultValue=0, description=""}
  User.CurrentExhaust {type=Vector3f, description="", defaultValue {struct=Vector3f, value=<0 0 0>}}
  User.ForwardBurst {type=NiagaraFloat, defaultValue=-1, description=""}
  User.ForwardDot {type=NiagaraFloat, defaultValue=0, description=""}
  User.PlayerSpeed {type=NiagaraFloat, defaultValue=0, description=""}
  User.RibbonIDOffset {type=NiagaraInt32}
  User.RightDot {type=NiagaraFloat, defaultValue=0, description=""}
  User.SkeletalMesh {
   type=NiagaraDataInterfaceSkeletalMesh
   defaultValue {
    struct=NiagaraExt_VariableValue_DataInterface
    value {dataInterface=&self:NiagaraDataInterfaceSkeletalMesh_0}
   }
  }
  User.TrailLengthScale {type=NiagaraFloat, defaultValue=1, description=""}
  User.Variant_Color {type=LinearColor, description="", defaultValue {struct=LinearColor, value=<rgba 1 0 0.5 1>}}
  User.Variant_Light {type=LinearColor, description="", defaultValue {struct=LinearColor, value=<rgba 1 0 0.5 1>}}
 ]
}
